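# Post-process-style AnimBlueprint for the Skull_Brite female player body, copying the attached parent mesh's pose and adding accessory motion.
# Weighted constraints place the knife and aim bones, then Look At orients the knife.
# AnimDynamics springs the bandana, additive offsets shift clavicle strap supports, and RigidBody simulates the physics asset relative to the right thigh twist.
# The event graph updates gravity-override and parachute variables, which the AnimGraph never reads.
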AnimBlueprint /Game/Characters/Player/Female/Medium/Bodies/F_MED_Skull_Brite/Meshes/F_MED_Skull_Brite_AnimBP
target_skeleton=&/F_MED_Skull_Brite_Skeleton
graphs [
 AnimGraph {nodes [
   n1 = "Constraint - dyn_knife (PPPP)"!(
    ComponentPose=Local To Component(LocalPose=Copy Pose From Mesh() {bUseAttachedParent=true})
    Alpha=0.675
    ConstraintWeights_0=0.0033
    ConstraintWeights_1:float=0.0136
    ConstraintWeights_2:float=0.0487
    ConstraintWeights_3:float=0.9343
   ) -> (Pose) {
    alphaBoolBlend {blendOption=Linear}
    boneToModify {boneName=dyn_knife}
    constraintSetup [
     {targetBone {boneName=upperarm_twist_01_l}}
     {targetBone {boneName=spine_04}}
     {targetBone {boneName=spine_05}}
     {targetBone {boneName=clavicle_l}}
    ]
    constraintWeights [1,1,1,1]
   }
   n2 = "Constraint - dyn_knife_aim (PPPPP)"!(
    ComponentPose=@n1
    Alpha=1
    ConstraintWeights_0=0.0001
    ConstraintWeights_1:float=0
    ConstraintWeights_2:float=0.7264
    ConstraintWeights_3:float=0.2234
    ConstraintWeights_4:float=0.0501
   ) -> (Pose) {
    alphaBoolBlend {blendOption=Linear}
    boneToModify {boneName=dyn_knife_aim}
    constraintSetup [
     {targetBone {boneName=clavicle_r}}
     {targetBone {boneName=upperarm_twist_01_l}}
     {targetBone {boneName=spine_04}}
     {targetBone {boneName=spine_05}}
     {targetBone {boneName=clavicle_l}}
    ]
    constraintWeights [1,1,1,1,1]
   }
   n3 = Look At - Bone: dyn_knife(ComponentPose=@n2, Alpha=1) {
    alphaBoolBlend {blendOption=Linear}
    boneToModify {boneName=dyn_knife}
    lookAtLocation=<0 0.1 5>
    lookAtTarget {boneReference {boneName=dyn_knife_aim}}
    lookAt_Axis {axis {y=0, z=1}}
    lookUp_Axis {axis=<0 1>}
   }
   n4 = Anim Dynamics - Bone: dyn_bandana(ComponentPose=@n3, Alpha=0.7, ExternalForce="0,0,0") {
    angularBiasOverride=0.9
    angularDampingOverride=0.995
    angularSpringConstant=150
    bAngularSpring=true
    bGravityOverrideInSimSpace=true
    bOverrideAngularBias=true
    bOverrideAngularDamping=true
    gravityScale=0.33
    simulationSpace=RootRelative
    alphaBoolBlend {blendOption=Linear}
    boundBone {boneName=dyn_bandana}
    physicsBodyDefinitions [{
      collisionType=CustomSphere
      sphereCollisionRadius=1
      boundBone {boneName=dyn_bandana}
      boxExtents=<10 10 10>
      constraintSetup {twistAxis=AxisZ, angularLimitsMax=<15 15 10>, angularLimitsMin=<-15 -15 -10>, angularTarget=<0 0 0>}
      localJointOffset=<0 0 -5>
     }]
    relativeSpaceBone {boneName=spine_05}
   }
   n5 = "Transform (Modify) Bone - Bone: clavicle_strap_support_l"!(ComponentPose=@n4, bAlphaBoolEnabled:bool=True, Translation="14.621,0.000,4.102", Rotation="0,0,0", Scale="1,1,1") -> (Pose) {
    alphaInputType=Bool
    rotationMode=BMM_Additive
    translationMode=BMM_Additive
    alphaBoolBlend {blendOption=Linear}
    boneToModify {boneName=clavicle_strap_support_l}
    translation {x=14.6208, z=4.1016}
   }
   n6 = "Transform (Modify) Bone - Bone: clavicle_strap_support_r"!(ComponentPose=@n5, bAlphaBoolEnabled:bool=True, Translation="-14.621,0,4.102", Rotation="0,0,0", Scale="1,1,1") -> (Pose) {
    alphaInputType=Bool
    rotationMode=BMM_Additive
    translationMode=BMM_Additive
    alphaBoolBlend {blendOption=Linear}
    boneToModify {boneName=clavicle_strap_support_r}
    translation {x=-14.621, z=4.102}
   }
   n7 = RigidBody!(ComponentPose=@n6, Alpha=0.7, OverrideWorldGravity:Vector="0,0,-980") -> (Pose) {
    bOverrideWorldGravity=true
    overridePhysicsAsset=&/F_MED_Skull_Brite_Physics
    simulationSpace=BaseBoneSpace
    alphaBoolBlend {blendOption=Linear}
    baseBoneRef {boneName=thigh_twist_01_r}
    componentAppliedLinearAccClamp=<8000 8000 8000>
    overrideWorldGravity {z=-980}
    simSpaceSettings {
     maxAngularAcceleration=400
     maxAngularVelocity=10
     maxLinearAcceleration=500
     maxLinearVelocity=1
     velocityScaleZ=0
     worldAlpha=1
     externalLinearDragV=<0.1 0.1>
     externalLinearVelocity=<0.1 0.1>
    }
   }
   Output Pose(Result=Component To Local(ComponentPose=@n7)) {layerGroup=None}
  ]}
 EventGraph {nodes [
   Event BlueprintUpdateAnimation()
   > Sequence()
   > n1 = Gravity Override State Function DEPRECATED(
    CharacterPartAnimInstance=Self-Reference()
    GameTimeInSeconds=GetGameTimeInSeconds()
    GravityOverrideParams=Get Gravity Override Params_0() -> (Gravity Override Params_0:Gravity Override Params Struct Structure)
   )
   > Set Gravity Override(Gravity Override=@n1)
   Event OnParachuteOpened()
   > Set Is Parachuting(Is Parachuting:bool=true) -> (Output_Get:bool)
   Event OnParachuteClosed()
   > Set Is Parachuting(Is Parachuting=false)
  ]}
]
variables [
 Gravity Override {type=Vector}
 Gravity Override Params_0 {type=GravityOverrideParamsStruct}
 Is Parachuting {type=boolean}
]
